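; Environment effect Blueprint for BlackMonday vent smoke that gives each placed instance a randomized look.
; Its construction script calls SetVisualParameters, which writes random width, pan, tile, curl and wave values into the smoke mesh's custom primitive data when cosmetic events can execute.
; Random height divided by mesh bounds sets X scale and also multiplies smoke tiling.
; Custom events re-roll one actor or every instance in the map.
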(event EventBeginPlay)

(event Collision|EventActorBeginOverlap
  (params (OtherActor :type Actor%20Object%20Reference)))

(event EventTick
  (params (DeltaSeconds :type Float%20%28single-precision%29)))

(event Custom|UpdateVisuals
  (CallFunction|SetVisualParameters))

(event Custom|UpdateVisualsOnAllActorsInMap
  (bind _outactors (Actor|GetAllActorsOfClass "/Figment_S10/Athena/Effects/Particles/Environments/BlackMonday/B_BlackMonday_VentSmoke.B_BlackMonday_VentSmoke_C"))
  (for _array_element _outactors
    (CallFunction|UpdateVisuals _array_element)))

(fn GetCurlIntensity
  (return (Utilities|RandomVectorinRange (Variables|Default|GetCurlIntensityMin) (Variables|Default|GetCurlIntensityMax))))

(fn GetCurlPanSpeedZ
  (return (Math|Random|RandomFloatinRange (Variables|Default|GetCurlPanSpeedMinZ) (Variables|Default|GetCurlPanSpeedMaxZ))))

(fn GetCurlTile
  (return (Math|Random|RandomFloatinRange (Variables|Default|GetCurlTileMin) (Variables|Default|GetCurlTileMax))))

(fn GetSmokePanSpeed
  (return (Math|Random|RandomFloatinRange (Variables|Default|GetSmokePanSpeedMin) (Variables|Default|GetSmokePanSpeedMax))))

(fn GetSmokeTile
  (return (* (Math|Random|RandomFloatinRange (Variables|Default|GetSmokeTileMin) (Variables|Default|GetSmokeTileMax)) (Variables|Default|GetSplineHeightRandomValue))))

(fn GetSplineHeight
  (bind (_min _max) (Components|StaticMesh|GetLocalBounds (Variables|Default|GetSmoke)))
  (bind _max_x (Math|Vector|BreakVector _max))
  (bind _output_get (Variables|Default|SetSplineHeightRandomValue (/ (Math|Random|RandomFloatinRange (Variables|Default|GetSplineHeightMin) (Variables|Default|GetSplineHeightMax)) _max_x)))
  (return _output_get))

(fn GetSplineWidthBase
  (return (Math|Random|RandomFloatinRange (Variables|Default|GetSplineWidthBaseMin) (Variables|Default|GetSplineWidthBaseMax))))

(fn GetSplineWidthTip
  (return (Math|Random|RandomFloatinRange (Variables|Default|GetSplineWidthTipMin) (Variables|Default|GetSplineWidthTipMax))))

(fn GetWaveSpeed
  (return (Math|Random|RandomFloatinRange (Variables|Default|GetSplineWaveSpeedMin) (Variables|Default|GetSplineWaveSpeedMax))))

(fn SetVisualParameters
  (bind (_vec3_x _vec3_y _vec3_z) (Math|Vector|BreakVector (CallFunction|GetCurlIntensity)))
  (Networking|CanExecuteCosmeticEvents
    (:True
      (Rendering|Material|SetCustomPrimitiveDataFloat (Variables|Default|GetSmoke) 0 (CallFunction|GetSplineWidthBase))
      (Rendering|Material|SetCustomPrimitiveDataFloat (Variables|Default|GetSmoke) 1 (CallFunction|GetSplineWidthTip))
      (Transformation|SetRelativeScale3D (Variables|Default|GetSmoke) (Math|Vector|MakeVector (CallFunction|GetSplineHeight) 1.0 1.0))
      (Rendering|Material|SetCustomPrimitiveDataFloat (Variables|Default|GetSmoke) 2 (CallFunction|GetSmokePanSpeed))
      (Rendering|Material|SetCustomPrimitiveDataFloat (Variables|Default|GetSmoke) 3 (CallFunction|GetSmokePanSpeed))
      (Rendering|Material|SetCustomPrimitiveDataFloat (Variables|Default|GetSmoke) 4 (CallFunction|GetSmokeTile))
      (Rendering|Material|SetCustomPrimitiveDataFloat (Variables|Default|GetSmoke) 5 (CallFunction|GetSmokeTile))
      (Rendering|Material|SetCustomPrimitiveDataFloat (Variables|Default|GetSmoke) 6 (CallFunction|GetCurlTile))
      (Rendering|Material|SetCustomPrimitiveDataFloat (Variables|Default|GetSmoke) 7 (CallFunction|GetCurlPanSpeedZ))
      (Rendering|Material|SetCustomPrimitiveDataVector4 (Variables|Default|GetSmoke) 8 (Math|Vector4|MakeVector4 _vec3_x _vec3_y _vec3_z))
      (Rendering|Material|SetCustomPrimitiveDataFloat (Variables|Default|GetSmoke) 12 (CallFunction|GetWaveSpeed)))))

(fn ConstructionScript
  (CallFunction|SetVisualParameters))
