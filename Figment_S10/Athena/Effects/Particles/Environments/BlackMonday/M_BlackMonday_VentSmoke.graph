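# Translucent, lit, two-sided material for drifting vent smoke in the BlackMonday environment effects, drawn on SplineThicken geometry.
# World position offset combines spline widening with scrolling curl noise and sine-driven rotation, masked along UV V.
# Two panned smoke textures are summed and shaped for color and opacity, with one end faded out.
# Custom primitive data tunes instances; depth and near-camera fades apply only at Medium and High quality.
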Material /Figment_S10/Athena/Effects/Particles/Environments/BlackMonday/M_BlackMonday_VentSmoke
graph {
 properties {
  blendMode=BLEND_Translucent
  pixelDepthOffsetMode=PDOM_Legacy
  translucencyLightingMode=TLM_VolumetricPerVertexDirectional
  twoSided=true
  usedShadingModels=Default Lit
 }
 nodes [
  n1 = ScalarParameter() {bUseCustomPrimitiveData=true, parameterName=CurlPanSpeedWS_Z, primitiveDataIndex=7}
  n2 = Subtract(
   A=WorldPosition().XYZ
   B=Multiply(A=Time(), B=MaterialFunctionCall(B=@n1) {materialFunction=&AppendMany}.RGB)
  )
  n3 = ScalarParameter() {bUseCustomPrimitiveData=true, parameterName=CurlTile, primitiveDataIndex=6}
  n4 = VectorParameter() {bUseCustomPrimitiveData=true, parameterName=CurlIntensity, primitiveDataIndex=8, defaultValue=<rgba 7 7 4.5 0>}
  n5 = Multiply(A=VectorNoise(World Position=Divide(A=@n2, B=@n3)) {noiseFunction=VNF_CurlALU}, B=@n4.RGB)
  n6 = Sine(None=Multiply(A=Time(), B=ScalarParameter() {bUseCustomPrimitiveData=true, parameterName=WaveSpeed, primitiveDataIndex=12}))
  n7 = RotateAboutAxis(
   NormalizedRotationAxis=Normalize(VectorInput=VectorParameter() {parameterName=RotationAxis, defaultValue=<rgba 1 0 0 0>}.RGB)
   RotationAngle=Multiply(A=ScalarParameter() {defaultValue=0.05, parameterName=RotationAngle}, B=ConstantBiasScale(None=@n6))
   PivotPoint=MaterialFunctionCall() {materialFunction=&ObjectPivotPoint}.Object Pivot Location
   Position=WorldPosition().XYZ
  )
  n8 = ComponentMask(None=TextureCoordinate()) {g=true}
  n9 = Constant()
  n10 = Constant() {r=1}
  n11 = Constant() {r=0.25}
  Add(
   A=MaterialFunctionCall(
    WidthBase=ScalarParameter() {bUseCustomPrimitiveData=true, defaultValue=15, parameterName=SplineWidthBase}
    WidthTip=ScalarParameter() {bUseCustomPrimitiveData=true, defaultValue=40, parameterName=SplineWidthTip, primitiveDataIndex=1}
   ) {materialFunction=&SplineThicken}.WorldPositionOffset
   B=Multiply(A=Add(A=@n5, B=@n7), B=SphereMask(A=@n8, B=@n9, Radius=@n10, Hardness=@n11))
  )
  n12 = VectorParameter() {parameterName=BaseColor, primitiveDataIndex=2, defaultValue=<rgba 0.4734 0.5573 0.6 0.05>}
  n13 = Constant() {r=1}
  n14 = ScalarParameter() {bUseCustomPrimitiveData=true, defaultValue=0.85, parameterName=SmokeTileA, primitiveDataIndex=4}
  n15 = Panner(
   Coordinate=Multiply(A=TextureCoordinate(), B=AppendVector(A=@n13, B=@n14))
   Speed=AppendVector(A=Constant(), B=ScalarParameter() {bUseCustomPrimitiveData=true, defaultValue=0.1, parameterName=SmokePanSpeedA, primitiveDataIndex=2})
  ) {bFractionalPart=true}
  n16 = TextureSample(UVs=VertexInterpolator(VS=MaterialFunctionCall(XY=@n15) {materialFunction=&Swizzle}.YX).PS) {texture=/Game/Effects/Fort_Effects/Textures/Smoke/T_Smoke_Trail_Soft_2}
  n17 = Constant() {r=1}
  n18 = ScalarParameter() {bUseCustomPrimitiveData=true, defaultValue=0.556, parameterName=SmokeTileB, primitiveDataIndex=5}
  n19 = Panner(
   Coordinate=Multiply(A=TextureCoordinate(), B=AppendVector(A=@n17, B=@n18))
   Speed=AppendVector(A=Constant(), B=ScalarParameter() {bUseCustomPrimitiveData=true, defaultValue=0.1288, parameterName=SmokePanSpeedB, primitiveDataIndex=3})
  ) {bFractionalPart=true}
  n20 = TextureSample(UVs=VertexInterpolator(VS=MaterialFunctionCall(XY=@n19) {materialFunction=&Swizzle}.YX).PS) {texture=&T_Smoke_Trail_Soft_2}
  n21 = ComponentMask(None=TextureCoordinate()) {g=true}
  n22 = Power(
   Base=Divide(A=Add(A=@n16.RGB, B=@n20.RGB))
   Exp=LinearInterpolate(B=ScalarParameter() {defaultValue=0.5, parameterName=Exponent}, Alpha=@n21) {constA=1}
  )
  Multiply(A=Multiply(A=@n12.RGB, B=@n22), B=@n12.A)
  n23 = Constant()
  n24 = Constant() {r=0.35}
  n25 = Constant()
  n26 = Multiply(
   A=Subtract(A=@n22, B=SphereMask(A=@n21, B=@n23, Radius=@n24, Hardness=@n25))
   B=ScalarParameter() {defaultValue=0.6, parameterName=OpacityScale}
  )
  n27 = Multiply(
   A=DepthFade(Opacity=@n26, FadeDistance=ScalarParameter() {defaultValue=20, parameterName=DepthFadeDistance})
   B=MaterialFunctionCall(Distance=ScalarParameter() {defaultValue=50, parameterName=NearCameraFadeDistance}) {materialFunction=&Near_Camera_Fade}.Result
  )
  QualitySwitch(Default=@n26, Low=@n26, High=@n27, Medium=@n27)
 ]
}
